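# Masked, lit material for Niagara mesh particles in the Chashu emote FX, built to deform and shade droplets.
# Dynamic parameters supply origin, position, rest normal and rigidity to bend vertices via RotateAboutAxis, weighted by a UV-based falloff mask.
# PreviousFrameSwitch repeats the bend with previous-position data for motion vectors.
# The same mask blends particle color, a fake cubemap reflection and hot-tip glow; a gradient wipe and debug nodes are also present.
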Material /BRCosmetics/Animation/Game/MainPlayer/Emotes/Chashu/FX/M_DemoDropletDeform
graph {
 properties {bUsedWithNiagaraMeshParticles=true, blendMode=BLEND_Masked, pixelDepthOffsetMode=PDOM_Legacy, usedShadingModels=Default Lit}
 nodes [
  n1 = ParticleColor()
  n2 = TextureSample!(UVs=ReflectionVectorWS(), Tex, Level=Constant() {r=5}, Apply View MipBias) {
   automaticViewMipBias=false
   constMipValue=3
   mipValueMode=TMVM_MipLevel
   samplerType=SAMPLERTYPE_LinearColor
   texture=/BRCosmetics/Environments/FrontEnd/Materials/BP12/Room_M_FE_Cubemap_Tex6
  }
  n3 = Multiply(A=@n1.RGB, B=Multiply(A=@n2.RGB, B=ScalarParameter() {defaultValue=0.25, parameterName=Refl}))
  n4 = Subtract(A=ComponentMask(None=TextureCoordinate()) {g=true}, B=ScalarParameter() {parameterName=Mask Offset})
  n5 = DynamicParameter() {parameterIndex=2, paramNames [MeshOriginX,MeshOriginY,MeshOriginZ,MaskFalloff]}
  n6 = Saturate(None=Power(Base=Saturate(None=@n4), Exp=@n5.MaskFalloff))
  n7 = Saturate(None=Power(Base=@n6, Exp=ScalarParameter() {defaultValue=9.5402}))
  n8 = LinearInterpolate(A=@n1.RGB, B=@n3, Alpha=@n7) {constB=0}
  LinearInterpolate(A=@n1.RGB, B=@n8, Alpha=@n6) {constA=1, constB=0}
  n9 = ParticlePositionWS()
  n10 = DynamicParameter() {parameterIndex=3, defaultValue=<rgba 1 1 1 0.15>, paramNames [PosX,PosY,PosZ,Rigidity]}
  n11 = MaterialFunctionCall(R=@n10.PosX, G=@n10.PosX, B=@n10.PosX) {materialFunction=&AppendMany}
  MaterialFunctionCall!(
   Vector3=Multiply(A=ScalarParameter() {defaultValue=1, parameterName=Param_2}, B=Subtract(A=@n9, B=@n11.RGB))
   MaximumNumberOfDigits
   UVs
   DebugTextLocation RG_UpperRight BA_LowerLeft
   Component Spacing
  ) {materialFunction=/Engine/Functions/Engine_MaterialFunctions02/Utility/DebugFloat3Values}
  MaterialFunctionCall() {materialFunction=&DebugTimeSine}
  MaterialFunctionCall(Emissive=Multiply(A=ParticleColor().RGB, B=ScalarParameter() {parameterName=Emissive Contribution})) {materialFunction=&MF_FX_EmissiveAdjustments}
  n12 = MaterialFunctionCall() {materialFunction=&LinearGradient}
  Multiply(A=@n12.VGradient)
  n13 = Add(
   A=OneMinus(None=@n12.VGradient)
   B=LinearInterpolate(Alpha=DynamicParameter() {paramNames [LerpPanner,PrevPosY,PrevPosZ,GlowAmt]}.LerpPanner) {constA=-1, desc=Resets values to go from -1 to 1}
  )
  n14 = ScalarParameter() {defaultValue=1, parameterName=BlowOutContrast}
  Reroute(None=MaterialFunctionCall(In=Clamp(None=@n13), Contrast=Multiply(A=@n14)) {materialFunction=&CheapContrast}.Result)
  n15 = DynamicParameter() {parameterIndex=1, defaultValue=<rgba 0 0 -1 1>, paramNames [RestNormalX, RestNormalY, RestNormalZ, --------------]}
  n16 = Multiply(A=MaterialFunctionCall(R=@n15.RestNormalX, G=@n15.RestNormalX, B=@n15.RestNormalX) {materialFunction=&AppendMany}.RGB) {constB=-1}
  n17 = MaterialFunctionCall(R=@n5.MeshOriginX, G=@n5.MeshOriginX, B=@n5.MeshOriginX, A=@n5.MeshOriginX) {materialFunction=&AppendMany}
  n18 = Subtract(A=@n11.RGB, B=@n17.RGB)
  n19 = WorldPosition()
  n20 = RotateAboutAxis(
   NormalizedRotationAxis=CrossProduct(A=@n16, B=Normalize(VectorInput=@n18))
   RotationAngle=@n10.Rigidity
   PivotPoint=@n17.RGB
   Position=@n19.XYZ
  )
  n21 = DynamicParameter() {paramNames [LerpPanner,PrevPosY,PrevPosZ,GlowAmt]}
  n22 = Subtract(
   A=MaterialFunctionCall(R=@n21.LerpPanner, G=@n21.LerpPanner, B=@n21.LerpPanner) {materialFunction=&AppendMany}.RGB
   B=@n17.RGB
  )
  n23 = RotateAboutAxis(
   NormalizedRotationAxis=CrossProduct(A=@n16, B=Normalize(VectorInput=@n22))
   RotationAngle=@n10.Rigidity
   PivotPoint=@n17.RGB
   Position=@n19.XYZ
  )
  Reroute(None=PreviousFrameSwitch(Current Frame=LinearInterpolate(A=@n20, Alpha=@n6) {constB=0}, Previous Frame=LinearInterpolate(A=@n23, Alpha=@n6) {constB=0}))
  n24 = DynamicParameter() {paramNames [LerpPanner,PrevPosY,PrevPosZ,GlowAmt]}
  n25 = ScalarParameter() {defaultValue=77.9017, parameterName=Hot Tip}
  n26 = Power(Base=@n1.RGB, Exp=ScalarParameter() {defaultValue=12.1724, parameterName=Param_1})
  n27 = Add(A=LinearInterpolate(B=Multiply(A=@n25, B=@n8), Alpha=Saturate(None=@n26)) {constB=0}, B=LinearInterpolate(B=@n3, Alpha=@n7) {constB=0})
  Reroute(None=LinearInterpolate(B=Multiply(A=@n24.GlowAmt, B=@n27), Alpha=@n6) {constB=0})
  ScalarParameter() {defaultValue=1, parameterName=Roughness}
  ScalarParameter() {defaultValue=1, parameterName=Falloff}
  ScalarParameter() {parameterName=Metalic}
  ScalarParameter() {parameterName=Spec}
 ]
}
