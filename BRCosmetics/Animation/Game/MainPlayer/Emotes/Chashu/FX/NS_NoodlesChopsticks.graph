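# Niagara effect for the Chashu emote that renders noodles dangling from chopsticks with dripping broth and food bits.
# NoodlesPendulum simulates one pendulum particle, its pivot blending from system position to a skeletal mesh sample via NoodleTransfer.
# It orients noodle meshes along the swing and feeds pivot, tip and NoodleEat into dynamic material parameters.
# DripSpawnRate scales the food-bit mesh, round-drop and foam sprite emitters; StraightDrops is disabled.
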NiagaraSystem /BRCosmetics/Animation/Game/MainPlayer/Emotes/Chashu/FX/NS_NoodlesChopsticks
dependencies {
 usedDataInterfaces [
  &NiagaraDataInterfaceMeshRendererInfo
  &NiagaraDataInterfaceColorCurve
  &NiagaraDataInterfaceCurve
  &NiagaraDataInterfaceArrayInt32
 ]
 usedDynamicInputs []
 usedModules [&SystemState]
}
emitters {
 FoodBits {
  emitterUpdateScript [
   &EmitterState(Inactive Response=::NewEnumerator1, Life Cycle Mode=::NewEnumerator1, Loop Duration=1, UseLoopDelay=true)
   &SpawnPerUnit(Module.Use Max Movement Threshold=true, Module.Use Movement Tolerance=false, Spawn Probability=0.75, Use Spawn Probability=true) {enabled=false}
   &SpawnRate(Spawn Probability=0.1, SpawnRate=&Multiply_Float(A=$User.DripSpawnRate, B=50), Use Spawn Probability=true)
  ]
  particleSpawnScript [
   &InitializeParticle(
    Color=<rgba 1 1 1 0.5>
    Lifetime Max=0.3
    Lifetime Min=0.2
    Lifetime Mode=::NewEnumerator1
    Mesh Scale Max=<0.2 0.2 0.2>
    Mesh Scale Min=<0.15 0.15 0.15>
    Mesh Scale Mode=::NewEnumerator2
    Sprite Size Mode=::NewEnumerator1
    Uniform Sprite Size Max=5
    Uniform Sprite Size Min=2
    UsePositionOffset=false
    Write Material Random=true
   )
   /SetVariables(Particles.MeshIndex=&SelectIntFromArray(
     Array Sampling Mode=::NewEnumerator1
     Direct Array Index=?
     Int Selection Array=DataInterface() {propertyValues {IntData [0,0,0,0,0,0,0,0,1,2,0,0,0,0]}}
     Random Seed=0
    )) {isSetParameters=true, module=SetVariables_#1}
   &InitialMeshOrientation(Mesh Orientation Mode=::NewEnumerator7, Rotation=<0 -0.25 0>, Use Rotation Vector=true)
   &SkeletalMeshLocation(
    "Inherit Velocity (Scale)":NiagaraFloat=&RandomRangeFloat(Minimum=-0.25, Maximum=-0.5)
    Inherited Velocity Speed Limit=200
    MaskSpawnGroup=false
    Mesh Sampling Type=::NewEnumerator5
    Orientation Sampling=::NewEnumerator1
    Sampled Position Offset=<0 -5 -5>
    Skeletal Mesh=$User.SkeletalMesh
    Socket Sampling Mode=::NewEnumerator5
    Velocity Sampling=::NewEnumerator0
    VelocitySpeedLimit=true
   ) {enabled=false}
   &SphereLocation(
    MaskSpawnGroup=false
    Non Uniform Scale=<1 1 0.3>
    Override Seed=false
    Sphere Orientation Axis:Vector3f=<1 0 0.3>
    Sphere Radius=2
    Surface Only=true
    Surface Only Band Thickness=0.1
    UseNonUniformScale=true
    UseOffset=false
    UseSphereOrientation=true
   )
   &AddVelocity(
    Coordinate Space=::Simulation
    Velocity=&Multiply_VectorByFloat(
     Vector=&RotateVectorByQuaternion(Vector To Rotate=<0 0 1>, Quaternion=$Particles.SkeletalMeshLocation.SampledRotation)
     Float=&RemapRange(
      Input Value To Remap=&VectorLength(Vector=$Particles.SkeletalMeshLocation.SampledVelocity)
      Input Min=0
      Input Max=500
      Output Min=0
      Output Max=200
     )
    )
   )
   &AddVelocityFromPoint(Module.Use Falloff Distance=false)
  ]
  particleUpdateScript [
   &ParticleState()
   &Drag(Drag=4, Use Linear Drag=true, Use Rotational Drag=false)
   &GravityForce(Gravity=<0 0 -800>)
   &SolveForcesAndVelocity(Clamp Velocity=false, Limit Acceleration=false)
   &ScaleColor(
    Scale Alpha=&FloatFromCurve(FloatCurve=DataInterface() {propertyValues {Curve {keys [
         {tangentMode=RCTM_User, time=0, value=1}
         {arriveTangent=-0.0405, interpMode=RCIM_Cubic, leaveTangent=-0.0405, tangentMode=RCTM_User, time=0.7009, value=1}
         {arriveTangent=-8.0089, interpMode=RCIM_Cubic, leaveTangent=-8.0089, tangentMode=RCTM_User, time=1, value=0}
        ]}}})
    ScaleA=true
    ScaleRGB=true
   )
   &ScaleMeshSizeBySpeed(Max Scale Factor=<1 1 2>, Min Scale Factor:Vector3f=<1 1 0.33>, Speed Threshold=200)
   &ScaleMeshSize(Scale Factor=&VectorFromFloat(Value=&FloatFromCurve(
      FloatCurve=DataInterface() {propertyValues {Curve {keys [
          <at 0 0.5 RCIM_Cubic>
          {interpMode=RCIM_Cubic, tangentMode=RCTM_User, time=0.05, value=1}
          {interpMode=RCIM_Cubic, tangentMode=RCTM_User, time=0.5, value=1}
          {arriveTangent=-1.4032, interpMode=RCIM_Cubic, leaveTangent=-1.4032, tangentMode=RCTM_User, time=1, value=0.5}
         ]}}}
      Scale Curve=0.8
     )))
   &UpdateMeshOrientation(Rotation Vector=<0 0 1>)
  ]
  rendererClasses [&NiagaraMeshRendererProperties]
  renderers [
   {rendererClass=&NiagaraMeshRendererProperties, rendererIndex=0}
   {rendererClass=&NiagaraMeshRendererProperties, rendererIndex=1}
  ]
 }
 FoodBits001 {
  emitterUpdateScript [
   &EmitterState(Inactive Response=::NewEnumerator1, Life Cycle Mode=::NewEnumerator1, Loop Duration=1, UseLoopDelay=true)
   &SpawnPerUnit(Module.Use Max Movement Threshold=true, Module.Use Movement Tolerance=false, Spawn Probability=0.75, Use Spawn Probability=true) {enabled=false}
   &SpawnRate(Spawn Probability=0.2, SpawnRate=&Multiply_Float(A=$User.DripSpawnRate, B=50), Use Spawn Probability=true)
  ]
  particleSpawnScript [
   &InitializeParticle(
    Color=<rgba 1 1 1 0.5>
    Lifetime Max=0.4
    Lifetime Min=0.2
    Lifetime Mode=::NewEnumerator1
    Mesh Scale Max=<0.05 0.05 0.05>
    Mesh Scale Min=<0.05 0.05 0.025>
    Mesh Scale Mode=::NewEnumerator2
    Sprite Size Mode=::NewEnumerator1
    Uniform Sprite Size Max=5
    Uniform Sprite Size Min=2
    UsePositionOffset=false
    Write Material Random=true
   )
   /SetVariables(Particles.MeshIndex=&RandomRangeInt(Minimum=0, Maximum=4)) {isSetParameters=true, module=SetVariables_#1}
   &InitialMeshOrientation(Mesh Orientation Mode=::NewEnumerator7, Rotation=<0 -0.25 0>, Use Rotation Vector=true)
   &SkeletalMeshLocation(
    "Inherit Velocity (Scale)":NiagaraFloat=&RandomRangeFloat(Minimum=-0.25, Maximum=-0.5)
    Inherited Velocity Speed Limit=200
    MaskSpawnGroup=false
    Mesh Sampling Type=::NewEnumerator5
    Orientation Sampling=::NewEnumerator1
    Sampled Position Offset=<0 -5 -5>
    Skeletal Mesh=$User.SkeletalMesh
    Socket Sampling Mode=::NewEnumerator5
    Velocity Sampling=::NewEnumerator0
    VelocitySpeedLimit=true
   ) {enabled=false}
   &SphereLocation(
    MaskSpawnGroup=false
    Non Uniform Scale=<1 1 0.3>
    Override Seed=false
    Sphere Orientation Axis:Vector3f=<1 0 0.3>
    Sphere Radius=2
    Surface Only=true
    Surface Only Band Thickness=0.1
    UseNonUniformScale=true
    UseOffset=false
    UseSphereOrientation=true
   )
   &AddVelocity(
    Coordinate Space=::Simulation
    Velocity=&Multiply_VectorByFloat(
     Vector=&RotateVectorByQuaternion(Vector To Rotate=<0 0 1>, Quaternion=$Particles.SkeletalMeshLocation.SampledRotation)
     Float=&RemapRange(
      Input Value To Remap=&VectorLength(Vector=$Particles.SkeletalMeshLocation.SampledVelocity)
      Input Min=0
      Input Max=500
      Output Min=0
      Output Max=200
     )
    )
   )
   &AddVelocityFromPoint(Module.Use Falloff Distance=false, Velocity Strength=80)
  ]
  particleUpdateScript [
   &ParticleState()
   &Drag(Drag=4, Use Linear Drag=true, Use Rotational Drag=false)
   &GravityForce(Gravity=<0 0 -800>)
   &SolveForcesAndVelocity(Clamp Velocity=false, Limit Acceleration=false)
   &ScaleColor(
    Scale Alpha=&FloatFromCurve(FloatCurve=DataInterface() {propertyValues {Curve {keys [
         {tangentMode=RCTM_User, time=0, value=1}
         {arriveTangent=-0.0405, interpMode=RCIM_Cubic, leaveTangent=-0.0405, tangentMode=RCTM_User, time=0.7009, value=1}
         {arriveTangent=-8.0089, interpMode=RCIM_Cubic, leaveTangent=-8.0089, tangentMode=RCTM_User, time=1, value=0}
        ]}}})
    ScaleA=true
    ScaleRGB=true
   )
   &ScaleMeshSizeBySpeed(Max Scale Factor=<1 1 2>, Min Scale Factor:Vector3f=<1 1 0.33>, Speed Threshold=200)
   &ScaleMeshSize(Scale Factor=&VectorFromFloat(Value=&FloatFromCurve(
      FloatCurve=DataInterface() {propertyValues {Curve {keys [
          <at 0 0.5 RCIM_Cubic>
          {interpMode=RCIM_Cubic, tangentMode=RCTM_User, time=0.05, value=1}
          {interpMode=RCIM_Cubic, tangentMode=RCTM_User, time=0.5, value=1}
          {arriveTangent=-1.4032, interpMode=RCIM_Cubic, leaveTangent=-1.4032, tangentMode=RCTM_User, time=1, value=0.5}
         ]}}}
      Scale Curve=0.8
     )))
  ]
  rendererClasses [&NiagaraMeshRendererProperties]
  renderers [
   {rendererClass=&NiagaraMeshRendererProperties, rendererIndex=0}
   {rendererClass=&NiagaraMeshRendererProperties, rendererIndex=1}
  ]
 }
 NoodlesPendulum {
  emitterUpdateScript [
   &EmitterState(
    Inactive Response=::NewEnumerator1
    Life Cycle Mode=::NewEnumerator1
    Loop Behavior=::NewEnumerator1
    Loop Duration Mode=::NewEnumerator1
    UseLoopDelay=false
   )
   &SpawnBurst_Instantaneous(Spawn Count=1, Use Spawn Probability=false)
  ]
  particleSpawnScript [
   /SetVariables(
    Particles.PendulumDrag:NiagaraFloat=&UniformRangedFloat(Minimum=0.5, Maximum=1)
    Particles.PendulumLengthScale:NiagaraFloat=&UniformRangedFloat(Minimum=0.75, Maximum=1)
    Particles.PendulumMaskFalloff=1
   ) {isSetParameters=true}
   /Niagara/Modules/Spawn/Initialization/InitializeParticle(
    Color=<rgba 0.6 0.5244 0.264 1>
    Lifetime=11
    Mesh Scale=&Multiply_Vector(A=<1 1 1>, B=&MakeVector(X=1, Y=1, Z=$Particles.PendulumLengthScale))
    Module.Write Color=true
    Module.Write Lifetime=true
    Module.Write Mass=true
    Module.Write Material Random=true
    Module.Write Position=false
    Module.Write Scale=true
    Module.Write SpriteRotation=false
    Module.Write SpriteSize=true
    Sprite Size=<1 1>
   )
   /SetVariables(Particles.RandomOffset:Vector3f=/Niagara/DynamicInputs/Random/RandomVector(
     Vector Scale:NiagaraFloat=&UniformRangedFloat(Minimum=0, Maximum=0)
     Constraint Vector:Vector3f=<0 0 1>
     Randomness Mode:ENiagaraRandomnessMode=::NewEnumerator0
     Random Seed:NiagaraInt32=0
     Fixed Random Seed:NiagaraBool=false
    )) {isSetParameters=true}
  ]
  particleUpdateScript [
   &ParticleState(Kill Particles When Lifetime Has Elapsed=false)
   &SkeletalMeshLocation(
    MaskSpawnGroup=false
    Mesh Sampling Type=::NewEnumerator5
    Orientation Sampling=::NewEnumerator1
    Position Sampling=::NewEnumerator2
    Sampled Position Offset=<-2.5 9 0>
    Skeletal Mesh=$User.SkeletalMesh
    Socket Sampling Mode=::NewEnumerator5
   )
   /SetVariables(Particles.MeshOrigin:Vector3f=&Lerp_Vector(
     A=&Add_Vector(A=/Niagara/DynamicInputs/Helpers/SimulationPosition(), B=$Particles.RandomOffset)
     B=Linked() {linkedVariable {name=Particles.SkeletalMeshLocation.SampledPosition, type=NiagaraPosition}}
     Alpha=$User.NoodleTransfer
    )) {isSetParameters=true}
   &GravityForce(Gravity=<0 0 -5000>)
   &Drag(Drag=$Particles.PendulumDrag, Use Linear Drag=true, Use Rotational Drag=false) {enabled=false}
   &PendulumSetup(
    Module.Enable Spring Driver=false
    Pendulum Length:NiagaraFloat=&Multiply_Float(A=18, B=$Particles.PendulumLengthScale)
    Pendulum Pivot:Vector3f=$Particles.MeshOrigin
    "Spring Driven Constraint (Tightness)"=3
   )
   &SolveForcesAndVelocity(Clamp Velocity=false, Limit Acceleration=false)
   &PendulumConstraint()
   &CalculateAccurateVelocity()
   /SetVariables(Particles.MeshFacing:Vector3f=&NormalizeVector(Vector To Normalize=&Subtract_Vector(A=$Particles.MeshOrigin, B=Linked() {linkedVariable {name=Particles.Position, type=NiagaraPosition}}))) {isSetParameters=true}
   /Niagara/Modules/Update/Orientation/V2/OrientMeshToVector(
    Look At Direction:Vector3f=&Multiply_VectorByFloat(Vector=$Particles.MeshFacing, Float=1)
    Override Quat=false
    Reference Axis Vector:Vector3f=<0 0 1>
    UseOptionalSideAxis=false
    UseOptionalUpAxis=false
   )
   &DynamicMaterialParameters(
    Index 0 Param 1=$User.NoodleEat
    Index 2 Param 1=HlslExpression() {hlslExpression=Particles.MeshOrigin.x}
    Index 2 Param 2=HlslExpression() {hlslExpression=Particles.MeshOrigin.y}
    Index 2 Param 3=HlslExpression() {hlslExpression=Particles.MeshOrigin.z}
    Index 2 Param 4=$Particles.PendulumMaskFalloff
    Index 3 Param 1=&MakeFloatFromVector(Vector=Linked() {linkedVariable {name=Particles.Position, type=NiagaraPosition}}, Channel=::NewEnumerator0)
    Index 3 Param 2=&MakeFloatFromVector(Vector=Linked() {linkedVariable {name=Particles.Position, type=NiagaraPosition}}, Channel=::NewEnumerator1)
    Index 3 Param 3=&MakeFloatFromVector(Vector=Linked() {linkedVariable {name=Particles.Position, type=NiagaraPosition}}, Channel=::NewEnumerator2)
    Index 3 Param 4=$Particles.PendulumRigidity
    Module.Param0WriteEnabled=true
    Module.Param0WriteEnabled2=true
    Module.Param0WriteEnabled3=true
    Module.Param1WriteEnabled=true
    Module.Param1WriteEnabled2=true
    Module.Param1WriteEnabled3=true
    Module.Param2WriteEnabled=true
    Module.Param2WriteEnabled2=true
    Module.Param2WriteEnabled3=true
    Module.Param3WriteEnabled=true
    Module.Param3WriteEnabled2=true
    Module.Param3WriteEnabled3=true
    Write Parameter Index 2=true
    Write Parameter Index 3=true
   )
  ]
  rendererClasses [&NiagaraSpriteRendererProperties, &NiagaraMeshRendererProperties]
  renderers [
   {rendererClass=&NiagaraSpriteRendererProperties}
   {rendererClass=&NiagaraMeshRendererProperties, rendererIndex=1}
   {rendererClass=&NiagaraMeshRendererProperties, rendererIndex=2}
   {rendererClass=&NiagaraMeshRendererProperties, rendererIndex=3}
   {rendererClass=&NiagaraMeshRendererProperties, rendererIndex=4}
  ]
 }
 RoundDrops {
  emitterUpdateScript [
   &EmitterState(Inactive Response=::NewEnumerator1, Life Cycle Mode=::NewEnumerator1, Loop Duration=1, UseLoopDelay=true)
   &SpawnPerUnit(
    Module.Use Max Movement Threshold=true
    Module.Use Movement Tolerance=false
    Spawn Probability=0.75
    Spawn Spacing=10
    Use Spawn Probability=true
   ) {enabled=false}
   &SpawnRate(SpawnRate=&Multiply_Float(A=$User.DripSpawnRate, B=25), Use Spawn Probability=true)
  ]
  particleSpawnScript [
   /Niagara/Modules/Spawn/Initialization/V2/InitializeParticle(
    Color=<rgba 1 1 1 0.5>
    Lifetime Max=0.15
    Lifetime Min=0.1
    Lifetime Mode=::NewEnumerator1
    Sprite Size Mode=::NewEnumerator1
    Uniform Sprite Size Max=2
    Uniform Sprite Size Min=1
    UsePositionOffset=false
    Write Material Random=true
   )
   &SkeletalMeshLocation(
    Bone / Socket Index=6
    "Inherit Velocity (Scale)":NiagaraFloat=&RandomRangeFloat(Minimum=1, Maximum=0.5)
    Inherited Velocity Speed Limit=200
    MaskSpawnGroup=false
    Mesh Sampling Type=::NewEnumerator5
    Orientation Sampling=::NewEnumerator1
    Sampled Position Offset=<0 -5 -5>
    Skeletal Mesh=$User.SkeletalMesh
    Socket Sampling Mode=::NewEnumerator5
    Velocity Sampling=::NewEnumerator0
    VelocitySpeedLimit=true
   ) {enabled=false}
   &SphereLocation(
    MaskSpawnGroup=false
    Non Uniform Scale=<1 1 0.3>
    Override Seed=false
    Sphere Orientation Axis:Vector3f=<1 0 0.3>
    Sphere Radius=5
    Surface Only=true
    Surface Only Band Thickness=0.1
    UseNonUniformScale=true
    UseOffset=false
    UseSphereOrientation=true
   )
   &AddVelocityFromPoint(Module.Use Falloff Distance=false, Velocity Strength=15)
   &AddVelocity(
    Coordinate Space=::Simulation
    Velocity=&Multiply_VectorByFloat(
     Vector=&RotateVectorByQuaternion(Vector To Rotate=<0 0 1>, Quaternion=$Particles.SkeletalMeshLocation.SampledRotation)
     Float=&RandomRangeFloat(Minimum=100, Maximum=300)
    )
   ) {enabled=false, module=AddVelocity001}
   &AddVelocity(Velocity=&Multiply_VectorByFloat(Vector=$Output.SphereLocation.SphereVector, Float=&RandomRangeFloat(Minimum=1, Maximum=3))) {enabled=false}
  ]
  particleUpdateScript [
   &ParticleState()
   &Drag(Use Linear Drag=true, Use Rotational Drag=false)
   &GravityForce()
   &SolveForcesAndVelocity(Clamp Velocity=false, Limit Acceleration=false)
   &ScaleColor(
    Scale Alpha=&FloatFromCurve(FloatCurve=DataInterface() {propertyValues {Curve {keys [
         {tangentMode=RCTM_User, time=0, value=1}
         {arriveTangent=-0.0405, interpMode=RCIM_Cubic, leaveTangent=-0.0405, tangentMode=RCTM_User, time=0.7009, value=1}
         {arriveTangent=-8.0089, interpMode=RCIM_Cubic, leaveTangent=-8.0089, tangentMode=RCTM_User, time=1, value=0}
        ]}}})
    ScaleA=true
    ScaleRGB=true
   )
   &ScaleSpriteSizeBySpeed(Max Scale Factor=<0.75 3>, Velocity Threshold=200)
   &ScaleSpriteSize(Uniform Curve Sprite Scale=DataInterface() {propertyValues {Curve {keys [
        {tangentMode=RCTM_User, time=0, value=1}
        {arriveTangent=-0.0405, interpMode=RCIM_Cubic, leaveTangent=-0.0405, tangentMode=RCTM_User, time=0.7009, value=1}
        {arriveTangent=-8.0089, interpMode=RCIM_Cubic, leaveTangent=-8.0089, tangentMode=RCTM_User, time=1, value=0}
       ]}}})
  ]
  rendererClasses [&NiagaraSpriteRendererProperties]
  renderers [{rendererClass=&NiagaraSpriteRendererProperties}]
 }
 Splash_Foam {
  emitterUpdateScript [
   &EmitterState(Inactive Response=::NewEnumerator1, Life Cycle Mode=::NewEnumerator1, Loop Behavior=::NewEnumerator1, Loop Duration=3600, UseLoopDelay=false)
   &SpawnRate(
    SpawnGroup:NiagaraInt32=&RandomRangeInt(Minimum=0, Maximum=1)
    SpawnRate=&Multiply_Float(A=$User.DripSpawnRate, B=60)
    Use Spawn Probability=false
   )
  ]
  particleSpawnScript [
   &InitializeParticle(
    Color=<rgba 1 1 1 0.5>
    Lifetime Max=0.3
    Lifetime Min=0.2
    Lifetime Mode=::NewEnumerator1
    Mesh Scale Mode=::NewEnumerator3
    Sprite Rotation Mode=::NewEnumerator2
    Sprite Size Mode=::NewEnumerator1
    Uniform Sprite Size Max=25
    Uniform Sprite Size Min=15
    UsePositionOffset=false
    Write Material Random=true
   ) {module=InitializeParticle001}
   /Niagara/Modules/Update/SubUV/SubUVAnimation(SubUV Animation Mode=::NewEnumerator1)
   &SkeletalMeshLocation(
    Bone Sampling Mode=::NewEnumerator4
    MaskSpawnGroup=false
    Orientation Sampling=::NewEnumerator1
    Sampled Position Offset=<0 15 0>
    Skeletal Mesh=$User.SkeletalMesh
   )
   &SphereLocation(
    MaskSpawnGroup=false
    Non Uniform Scale=<1 1 0.3>
    Override Seed=false
    Sphere Orientation Axis:Vector3f=<1 0 0.3>
    Sphere Radius=2
    Surface Only=true
    Surface Only Band Thickness=0.1
    UseNonUniformScale=true
    UseOffset=false
    UseSphereOrientation=true
   )
   &AddVelocity(
    Coordinate Space=::Simulation
    Velocity=&Multiply_VectorByFloat(
     Vector=&RotateVectorByQuaternion(Vector To Rotate=<0 0 1>, Quaternion=$Particles.SkeletalMeshLocation.SampledRotation)
     Float=&RemapRange(
      Input Value To Remap=&VectorLength(Vector=$Particles.SkeletalMeshLocation.SampledVelocity)
      Input Min=0
      Input Max=100
      Output Min=0
      Output Max=200
     )
    )
   )
   &AddVelocityFromPoint(Module.Use Falloff Distance=false, Velocity Strength=20)
  ]
  particleUpdateScript [
   &ParticleState()
   &SpriteRotationRate(Rotation Rate=&RandomRangeFloat(Minimum=-90, Maximum=90))
   &Drag(Drag=0.5, Use Linear Drag=true, Use Rotational Drag=false)
   &GravityForce(Gravity=<0 0 -900>)
   &SolveForcesAndVelocity(Clamp Velocity=false, Limit Acceleration=false)
   &ScaleSpriteSize(Scale Factor=&Vector2DFromFloat(Value=&FloatFromCurve(
      FloatCurve=DataInterface() {propertyValues {Curve {keys [
          {arriveTangent=1.6115, interpMode=RCIM_Cubic, leaveTangent=1.6115, tangentMode=RCTM_User, time=0, value=0.33}
          {interpMode=RCIM_Cubic, tangentMode=RCTM_User, time=1, value=1}
         ]}}}
      Scale Curve=0.65
     )))
   &DynamicMaterialParameters(
    Index 0 Param 1=&RandomRangeFloat(Minimum=0, Maximum=1)
    Index 0 Param 3=&FloatFromCurve(FloatCurve=DataInterface() {propertyValues {Curve {keys [{interpMode=RCIM_Cubic, tangentMode=RCTM_User, time=0, value=0}, {interpMode=RCIM_Cubic, tangentMode=RCTM_User, time=1, value=1}]}}})
    Index 0 Param 4=&FloatFromCurve(FloatCurve=DataInterface() {propertyValues {Curve {keys [
         {arriveTangent=1.9228, interpMode=RCIM_Cubic, leaveTangent=1.9228, tangentMode=RCTM_User, time=0, value=0}
         {arriveTangent=1.5276, interpMode=RCIM_Cubic, leaveTangent=1.5276, tangentMode=RCTM_User, time=1, value=1}
        ]}}})
    Module.Param0WriteEnabled=true
    Module.Param1WriteEnabled=false
    Module.Param2WriteEnabled=true
    Module.Param3WriteEnabled=true
   )
  ]
  rendererClasses [&NiagaraSpriteRendererProperties]
  renderers [{rendererClass=&NiagaraSpriteRendererProperties}]
 }
 StraightDrops {
  bEnabled=false
  emitterUpdateScript [
   &EmitterState(Inactive Response=::NewEnumerator1, Life Cycle Mode=::NewEnumerator1, Loop Duration=1, UseLoopDelay=true)
   &SpawnPerUnit(Module.Use Max Movement Threshold=true, Module.Use Movement Tolerance=false, Spawn Probability=0.75, Use Spawn Probability=true) {enabled=false}
   &SpawnRate(SpawnRate=&Multiply_Float(A=$User.DripSpawnRate, B=80), Use Spawn Probability=true)
  ]
  particleSpawnScript [
   &InitializeParticle(
    Color=<rgba 1 1 1 0.5>
    Lifetime Max=0.3
    Lifetime Min=0.15
    Lifetime Mode=::NewEnumerator1
    Mesh Scale Max=<0.1 0.1 0.15>
    Mesh Scale Min=<0.05 0.05 0.025>
    Mesh Scale Mode=::NewEnumerator2
    Sprite Size Mode=::NewEnumerator1
    Uniform Sprite Size Max=5
    Uniform Sprite Size Min=2
    UsePositionOffset=false
    Write Material Random=true
   )
   &InitialMeshOrientation(Mesh Orientation Mode=::NewEnumerator7, Rotation=<0 -0.25 0>, Use Rotation Vector=true)
   &SkeletalMeshLocation(
    "Inherit Velocity (Scale)":NiagaraFloat=&RandomRangeFloat(Minimum=-0.25, Maximum=-0.5)
    Inherited Velocity Speed Limit=200
    MaskSpawnGroup=false
    Mesh Sampling Type=::NewEnumerator5
    Orientation Sampling=::NewEnumerator1
    Sampled Position Offset=<0 -5 -5>
    Skeletal Mesh=$User.SkeletalMesh
    Socket Sampling Mode=::NewEnumerator5
    Velocity Sampling=::NewEnumerator0
    VelocitySpeedLimit=true
   ) {enabled=false}
   &SphereLocation(
    MaskSpawnGroup=false
    Non Uniform Scale=<1 1 0.3>
    Override Seed=false
    Sphere Orientation Axis:Vector3f=<1 0 0.3>
    Sphere Radius=2
    Surface Only=true
    Surface Only Band Thickness=0.1
    UseNonUniformScale=true
    UseOffset=false
    UseSphereOrientation=true
   )
   &AddVelocity(
    Coordinate Space=::Simulation
    Velocity=&Multiply_VectorByFloat(
     Vector=&RotateVectorByQuaternion(Vector To Rotate=<0 0 1>, Quaternion=$Particles.SkeletalMeshLocation.SampledRotation)
     Float=&RemapRange(
      Input Value To Remap=&VectorLength(Vector=$Particles.SkeletalMeshLocation.SampledVelocity)
      Input Min=0
      Input Max=500
      Output Min=0
      Output Max=200
     )
    )
   )
   &AddVelocityFromPoint(Module.Use Falloff Distance=false, Velocity Strength=15)
  ]
  particleUpdateScript [
   &ParticleState()
   &Drag(Drag=4, Use Linear Drag=true, Use Rotational Drag=false)
   &GravityForce(Gravity=<0 0 -700>)
   &SolveForcesAndVelocity(Clamp Velocity=false, Limit Acceleration=false)
   &ScaleColor(
    Scale Alpha=&FloatFromCurve(FloatCurve=DataInterface() {propertyValues {Curve {keys [
         {tangentMode=RCTM_User, time=0, value=1}
         {arriveTangent=-0.0405, interpMode=RCIM_Cubic, leaveTangent=-0.0405, tangentMode=RCTM_User, time=0.7009, value=1}
         {arriveTangent=-8.0089, interpMode=RCIM_Cubic, leaveTangent=-8.0089, tangentMode=RCTM_User, time=1, value=0}
        ]}}})
    ScaleA=true
    ScaleRGB=true
   )
   &ScaleMeshSizeBySpeed(Max Scale Factor=<1 1 2>, Min Scale Factor:Vector3f=<1 1 0.33>, Speed Threshold=200)
   &ScaleMeshSize(Scale Factor=&VectorFromFloat(Value=&FloatFromCurve(
      FloatCurve=DataInterface() {propertyValues {Curve {keys [
          <at 0 0.5 RCIM_Cubic>
          {interpMode=RCIM_Cubic, tangentMode=RCTM_User, time=0.05, value=1}
          {interpMode=RCIM_Cubic, tangentMode=RCTM_User, time=0.5, value=1}
          {arriveTangent=-1.4032, interpMode=RCIM_Cubic, leaveTangent=-1.4032, tangentMode=RCTM_User, time=1, value=0.5}
         ]}}}
      Scale Curve=0.8
     )))
  ]
  rendererClasses [&NiagaraMeshRendererProperties]
  renderers [{rendererClass=&NiagaraMeshRendererProperties, rendererIndex=0}]
 }
}
summary {
 systemName=NS_NoodlesChopsticks
 emitters [
  {
   emitterName=NoodlesPendulum
   rendererClasses [&NiagaraSpriteRendererProperties, &NiagaraMeshRendererProperties]
  }
  {bEnabled=false, emitterName=StraightDrops, rendererClasses [&NiagaraMeshRendererProperties]}
  {emitterName=RoundDrops, rendererClasses [&NiagaraSpriteRendererProperties]}
  {emitterName=FoodBits, rendererClasses [&NiagaraMeshRendererProperties]}
  {emitterName=FoodBits001, rendererClasses [&NiagaraMeshRendererProperties]}
  {emitterName=Splash_Foam, rendererClasses [&NiagaraSpriteRendererProperties]}
 ]
 userVariables [
  User.DripSpawnRate {type=NiagaraFloat, defaultValue=1, description=""}
  User.ForwardDot {type=NiagaraFloat, defaultValue=0, description=""}
  User.NoodleEat {type=NiagaraFloat, defaultValue=1, description=""}
  User.NoodleTransfer {type=NiagaraFloat, defaultValue=0, description=""}
  User.RightDot {type=NiagaraFloat, defaultValue=0, description=""}
  User.SkeletalMesh {
   type=NiagaraDataInterfaceSkeletalMesh
   defaultValue {
    struct=NiagaraExt_VariableValue_DataInterface
    value {dataInterface=&self:NiagaraDataInterfaceSkeletalMesh_0}
   }
  }
 ]
}
