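# Niagara effect for the Chashu emote that dangles swinging noodle meshes from the player toward the mouth.
# A single emitter spawns noodles, gated by User.NoodleEat, as gravity-driven pendulums with randomized lengths.
# Each pivot blends from system position toward an offset skeletal mesh sample via clamped User.NoodleTransfer.
# Meshes orient to their pivot, while dynamic material parameters receive eat amount, pivot, bob position, falloff and rigidity.
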NiagaraSystem /BRCosmetics/Animation/Game/MainPlayer/Emotes/Chashu/FX/NS_NoodlesToMouth
dependencies {usedDataInterfaces [&NiagaraDataInterfaceCurve], usedDynamicInputs [], usedModules [&SystemState]}
emitters {NoodlesPendulum {
  emitterUpdateScript [
   &EmitterState(
    Inactive Response=::NewEnumerator1
    Life Cycle Mode=::NewEnumerator1
    Loop Behavior=::NewEnumerator1
    Loop Duration Mode=::NewEnumerator1
    UseLoopDelay=false
   )
   &SpawnBurst_Instantaneous(Spawn Count=1, Use Spawn Probability=false)
   &SpawnRate(Spawn Probability=&Multiply_Float(A=$User.NoodleEat, B=1), SpawnRate=2, Use Spawn Probability=true)
  ]
  particleSpawnScript [
   /SetVariables(
    Particles.PendulumDrag:NiagaraFloat=&UniformRangedFloat(Minimum=1, Maximum=1)
    Particles.PendulumLengthScale:NiagaraFloat=&UniformRangedFloat(Minimum=0.75, Maximum=1)
    Particles.PendulumMaskFalloff=1
    Particles.PendulumRigidity=0.15
   ) {isSetParameters=true}
   /Niagara/Modules/Spawn/Initialization/InitializeParticle(
    Color=<rgba 0.6 0.5244 0.264 1>
    Lifetime=8
    Mesh Scale=&Multiply_Vector(A=<1 1 1>, B=&MakeVector(X=1, Y=1, Z=$Particles.PendulumLengthScale))
    Module.Write Color=true
    Module.Write Lifetime=true
    Module.Write Mass=true
    Module.Write Material Random=true
    Module.Write Position=false
    Module.Write Scale=true
    Module.Write SpriteRotation=false
    Module.Write SpriteSize=true
    Sprite Size=<40 40>
   )
   /SetVariables(Particles.RandomOffset:Vector3f=/Niagara/DynamicInputs/Random/RandomVector(
     Vector Scale:NiagaraFloat=&UniformRangedFloat(Minimum=0, Maximum=0)
     Constraint Vector:Vector3f=<0 0 1>
     Randomness Mode:ENiagaraRandomnessMode=::NewEnumerator0
     Random Seed:NiagaraInt32=0
     Fixed Random Seed:NiagaraBool=false
    )) {isSetParameters=true}
   /SetVariables(Particles.Color=&MultiplyLinearColor(
     Linear Color=&HueShiftLinearColor(
      Hue=0
      Saturation=&RandomRangeFloat(Minimum=1, Maximum=1.2)
      Value=&RandomRangeFloat(Minimum=0.5, Maximum=1.2)
      Linear Color=$Particles.Color
     )
     Scale Factor=<1 1 1 1>
    )) {enabled=false, isSetParameters=true}
   &InitialMeshOrientation(
    Mesh Orientation Mode=::NewEnumerator7
    Rotation=&RandomRangeVector(Minimum=<0 0 0>, Maximum=<0 0 1>)
    Use Rotation Vector=true
   )
   &CameraOffset(Camera Offset Amount=1)
  ]
  particleUpdateScript [
   &ParticleState()
   &SkeletalMeshLocation(
    Bone Sampling Mode=::NewEnumerator4
    MaskSpawnGroup=false
    Orientation Sampling=::NewEnumerator1
    Position Sampling=::NewEnumerator2
    Sampled Position Offset=&Multiply_VectorByFloat(Vector=<-1.5 7 0>, Float=1)
    Skeletal Mesh=$User.SkeletalMesh
   )
   /SetVariables(Particles.MeshOrigin:Vector3f=&Lerp_Vector(
     A=/Niagara/DynamicInputs/Helpers/SimulationPosition()
     B=Linked() {linkedVariable {name=Particles.SkeletalMeshLocation.SampledPosition, type=NiagaraPosition}}
     Alpha=&ClampFloat(Float=$User.NoodleTransfer, Min=0.8, Max=0.9)
    )) {isSetParameters=true}
   &GravityForce(Gravity=<0 0 -5000>)
   &Drag(Drag=$Particles.PendulumDrag, Use Linear Drag=true, Use Rotational Drag=false)
   &PendulumSetup(
    Module.Enable Spring Driver=false
    Pendulum Length:NiagaraFloat=&Multiply_Float(A=18, B=$Particles.PendulumLengthScale)
    Pendulum Pivot:Vector3f=$Particles.MeshOrigin
    "Spring Driven Constraint (Tightness)"=3
   )
   &SolveForcesAndVelocity(Clamp Velocity=false, Limit Acceleration=false)
   &PendulumConstraint()
   &CalculateAccurateVelocity()
   /SetVariables(Particles.MeshFacing:Vector3f=&NormalizeVector(Vector To Normalize=&Subtract_Vector(A=$Particles.MeshOrigin, B=Linked() {linkedVariable {name=Particles.Position, type=NiagaraPosition}}))) {isSetParameters=true}
   /Niagara/Modules/Update/Orientation/V2/OrientMeshToVector(
    Look At Direction:Vector3f=&Multiply_VectorByFloat(Vector=$Particles.MeshFacing, Float=1)
    Override Quat=false
    Reference Axis Vector:Vector3f=<0 0 1>
    UseOptionalSideAxis=false
    UseOptionalUpAxis=false
   )
   &DynamicMaterialParameters(
    Index 0 Param 1=$User.NoodleEat
    Index 2 Param 1=HlslExpression() {hlslExpression=Particles.MeshOrigin.x}
    Index 2 Param 2=HlslExpression() {hlslExpression=Particles.MeshOrigin.y}
    Index 2 Param 3=HlslExpression() {hlslExpression=Particles.MeshOrigin.z}
    Index 2 Param 4=$Particles.PendulumMaskFalloff
    Index 3 Param 1=&MakeFloatFromVector(Vector=Linked() {linkedVariable {name=Particles.Position, type=NiagaraPosition}}, Channel=::NewEnumerator0)
    Index 3 Param 2=&MakeFloatFromVector(Vector=Linked() {linkedVariable {name=Particles.Position, type=NiagaraPosition}}, Channel=::NewEnumerator1)
    Index 3 Param 3=&MakeFloatFromVector(Vector=Linked() {linkedVariable {name=Particles.Position, type=NiagaraPosition}}, Channel=::NewEnumerator2)
    Index 3 Param 4=$Particles.PendulumRigidity
    Module.Param0WriteEnabled=true
    Module.Param0WriteEnabled2=true
    Module.Param0WriteEnabled3=true
    Module.Param1WriteEnabled=true
    Module.Param1WriteEnabled2=true
    Module.Param1WriteEnabled3=true
    Module.Param2WriteEnabled=true
    Module.Param2WriteEnabled2=true
    Module.Param2WriteEnabled3=true
    Module.Param3WriteEnabled=true
    Module.Param3WriteEnabled2=true
    Module.Param3WriteEnabled3=true
    Write Parameter Index 2=true
    Write Parameter Index 3=true
   )
   &ScaleMeshSize(Scale Factor=&MakeVector(
     X=1
     Y=1
     Z=&Multiply_Float(
      A=&FloatFromCurve(FloatCurve=DataInterface() {propertyValues {Curve {keys [
           {interpMode=RCIM_Cubic, tangentMode=RCTM_User, time=0, value=0.8}
           {interpMode=RCIM_Cubic, tangentMode=RCTM_User, time=0.5, value=1}
           {tangentMode=RCTM_User, time=1, value=0.8}
          ]}}})
      B=1
     )
    ))
  ]
  rendererClasses [&NiagaraSpriteRendererProperties, &NiagaraMeshRendererProperties]
  renderers [
   {rendererClass=&NiagaraSpriteRendererProperties}
   {rendererClass=&NiagaraMeshRendererProperties, rendererIndex=1}
   {rendererClass=&NiagaraMeshRendererProperties, rendererIndex=2}
   {rendererClass=&NiagaraMeshRendererProperties, rendererIndex=3}
   {rendererClass=&NiagaraMeshRendererProperties, rendererIndex=4}
   {rendererClass=&NiagaraMeshRendererProperties, rendererIndex=5}
   {rendererClass=&NiagaraMeshRendererProperties, rendererIndex=6}
  ]
 }}
summary {
 systemName=NS_NoodlesToMouth
 emitters [{
   emitterName=NoodlesPendulum
   rendererClasses [&NiagaraSpriteRendererProperties, &NiagaraMeshRendererProperties]
  }]
 userVariables [
  User.DripSpawnRate {type=NiagaraFloat, defaultValue=1, description=""}
  User.NoodleEat {type=NiagaraFloat, defaultValue=1, description=""}
  User.NoodleTransfer {type=NiagaraFloat, defaultValue=0, description=""}
  User.SkeletalMesh {
   type=NiagaraDataInterfaceSkeletalMesh
   defaultValue {
    struct=NiagaraExt_VariableValue_DataInterface
    value {dataInterface=&self:NiagaraDataInterfaceSkeletalMesh_0}
   }
  }
 ]
}
